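# Translucent UI material for an Ice-themed Item Shop tile, with colors from the shared background-parameter function.
# Ice and frost textures supply masks, with panning noise and Divide-based brightening adding variation to the frosted color.
# A rotated reflected gradient, animated with delay and cubic easing, forms a reflection mask.
# UV scaling fits the effect to the tile, and rounded-corner functions using the shop's CornerRadius collection parameter shape the mask and outline.
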Material /ItemShop/UI/Art/Ice/M_UI_ShopTile_Ice
graph {
 properties {bAutomaticallySetUsageInEditor=false, blendMode=BLEND_Translucent, materialDomain=MD_UI, usedShadingModels=Default Lit}
 nodes [
  n1 = NamedRerouteUsage()
  n2 = NamedRerouteUsage()
  n3 = Add(A=Multiply(A=NamedRerouteUsage()), B=Panner(Coordinate=NamedRerouteUsage()) {speedX=0.01, speedY=-0.02})
  n4 = TextureObject() {samplerType=SAMPLERTYPE_Alpha, texture=&/T_UI_Figment_07_Noise}
  n5 = Divide(
   A=Multiply(A=@n1, B=LinearInterpolate(Alpha=@n2) {constA=1}) {desc=Neutralize Center}
   B=LinearInterpolate(Alpha=TextureSample(UVs=@n3, Tex=@n4).R) {constA=0.01}
  ) {constB=0.2}
  n6 = NamedRerouteUsage()
  n7 = Constant() {r=1.5}
  n8 = NamedRerouteDeclaration(None=TextureSample(
    UVs=MaterialFunctionCall(UVs=@n6, Texture Scale=@n7) {materialFunction=&ScaleUVsByCenter}.UVs
    Tex=TextureObject() {samplerType=SAMPLERTYPE_LinearColor, texture=&/T_IceGauntlets_Ice}
   ).R) {name=Ice Mask, nodeColor=<rgba 0 0 0 1>}
  n9 = NamedRerouteUsage()
  n10 = Constant() {r=0.1}
  n11 = Constant() {r=0.6}
  n12 = Divide(A=Saturate(None=MaterialFunctionCall!(Input=@n9, Input Low=@n10, Input High=@n11) {materialFunction=/Engine/Functions/Engine_MaterialFunctions03/Math/RemapValueRangeNormalized}.Result)) {constB=0.5}
  n13 = NamedRerouteUsage()
  n14 = Divide(A=Saturate(None=LinearInterpolate(A=@n1, B=@n5, Alpha=@n8)), B=LinearInterpolate(Alpha=Multiply(A=@n12, B=@n13)) {constA=1, constB=0.6})
  n15 = LinearInterpolate(Alpha=NamedRerouteUsage()) {constA=0.33, constB=0.5}
  n16 = NamedRerouteUsage()
  n17 = SmoothStep(Value=OneMinus(None=ComponentMask(None=@n16) {g=true}))
  n18 = Multiply(A=NamedRerouteUsage()) {constB=-1}
  n19 = Abs(None=Multiply(A=NamedRerouteUsage(), B=Multiply(A=@n18) {constB=2}))
  n20 = Constant() {r=12}
  n21 = SmoothStep(
   Max=LinearInterpolate(Alpha=@n17) {constB=0.75}
   Value=Multiply(A=@n19, B=LinearInterpolate(B=@n20, Alpha=@n17) {constA=12, constB=8}) {constA=1, constB=64}
  ) {constMax=0.5}
  n22 = Saturate(None=Divide(A=Saturate(None=@n14), B=LinearInterpolate(A=@n15, Alpha=@n21) {constA=0.125}))
  n23 = SmoothStep(Value=@n19) {constMax=0.05}
  LinearInterpolate(A=@n22, B=Saturate(None=Add(A=@n22) {constB=0.25}), Alpha=Multiply(A=OneMinus(None=@n23), B=@n17)) {desc=Frost Effect}
  MaterialFunctionCall!(Input Alpha=Constant() {r=1}, bOverrideCornerRadius, Corner Radius Override) {materialFunction=/FortUILibrary/Systems/ItemShopTile/Materials/MaterialFunctions/MF_UI_ItemShop_RoundedCorners}
  n24 = CollectionParameter() {collection=/FortUILibrary/Systems/ItemShopTile/MPC_ItemShop, parameterName=CornerRadius}
  n25 = MaterialFunctionCall(
   Corner Radius=MaterialFunctionCall(R=@n24, G=@n24, B=@n24, A=@n24) {materialFunction=&AppendMany}.RGBA
   Radius in Pixels=Constant() {r=1}
   Outline Thickness=Constant() {r=4}
  ) {materialFunction=&MF_UI_CoreUI_RoundedMask}
  Multiply(A=OneMinus(None=SmoothStep(Value=@n25.Outline SDF) {constMax=0.25})) {constB=0.2}
  NamedRerouteDeclaration(None=LinearInterpolate(A=NamedRerouteUsage(), B=NamedRerouteUsage(), Alpha=NamedRerouteUsage())) {name=Background Color, nodeColor=<rgba 0 0 0 1>}
  n26 = MaterialFunctionCall() {materialFunction=&MF_UI_UVScaling}
  NamedRerouteDeclaration(None=OneMinus(None=ComponentMask(None=@n26.Scaled to Fit) {g=true})) {name=Focal Mask UV, nodeColor=<rgba 0 0 0 1>}
  NamedRerouteDeclaration(None=@n26.ScaleY) {name=Scale Y UV, nodeColor=<rgba 0 0 0 1>}
  n27 = Rotator(Coordinate=NamedRerouteUsage(), Time=Constant() {r=-0.1}) {speed=6.2832}
  n28 = Constant() {r=8}
  n29 = Constant() {r=2}
  n30 = LinearInterpolate(Alpha=MaterialFunctionCall("Value (0-1)"=MaterialFunctionCall!(Animation Duration=@n28, Delay=@n29, Reset to Default) {materialFunction=/FortUILibrary/MaterialFunctions/Time/MF_UI_AnimationDelay}.Result) {materialFunction=&MF_UI_EaseCurves}.EaseInCubic) {constA=-2, constB=2}
  NamedRerouteDeclaration(None=SmoothStep(Value=MaterialFunctionCall!(UVs=@n27, Offset=@n30, RotationCenter, RotationAngle) {materialFunction=/FortUILibrary/MaterialFunctions/Gradients/MF_UI_ReflectedGradient}.Result) {constMin=0.5}) {name=Reflection Mask, nodeColor=<rgba 0 0 0 1>}
  NamedRerouteDeclaration(None=@n25.SDF) {name=OutlineFrame, nodeColor=<rgba 0 0.5547 1 1>}
  NamedRerouteDeclaration(None=@n26.Scaled to Fit) {name=UV, nodeColor=<rgba 0 0 0 1>}
  n31 = MaterialFunctionCall() {materialFunction=/FortUILibrary/Systems/ItemShopTile/Materials/MaterialFunctions/MF_UI_ItemShop_BackgroundParams}
  NamedRerouteDeclaration(None=@n31.Color 1) {name=color 1, nodeColor=<rgba 0.2422 0 1 1>}
  n32 = NamedRerouteUsage()
  n33 = Constant() {r=0.75}
  n34 = MaterialFunctionCall(
   Input=TextureSample(
    UVs=MaterialFunctionCall(UVs=@n32, Texture Scale=@n33) {materialFunction=&ScaleUVsByCenter}.UVs
    Tex=TextureObject() {samplerType=SAMPLERTYPE_Alpha, texture=&/T_UI_Figment_07_Frost}
   ) {samplerType=SAMPLERTYPE_Alpha}.R
   Input Low=Constant() {r=0.3}
   Input High=Constant() {r=1}
  ) {materialFunction=&RemapValueRangeNormalized}
  NamedRerouteDeclaration(None=Saturate(None=LinearInterpolate(Alpha=@n34.Result) {constB=1.5})) {name=Frost Mask, nodeColor=<rgba 0 0 0 1>}
  NamedRerouteDeclaration(None=@n31.Color 2) {name=color 2, nodeColor=<rgba 0.0781 1 0 1>}
  NamedRerouteDeclaration(None=@n31.Color 3) {name=color 3, nodeColor=<rgba 1 0 0.3984 1>}
 ]
}
